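; Season 26 Battle Pass quest ability that tracks time spent crouched near opponents.
; Once the crouched tag appears, a repeating timer scans nearby player pawns for a matching tag, accumulating time on consecutive successes and otherwise sending NotInRange.
; Past a threshold, floored time goes to the quest manager and a CrouchNearEnemy verb is broadcast, carrying leftovers forward.
; Missing required tags stops the timer; DBNO cancels the ability.
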(event Ability|EventActivateAbility
  ;[1 1 1] Run first check and start looping check
  (|CallParentFunction)
  (Ability|Tasks|WaitGameplayTagAdd (Variables|Default|GetPlayerCrouched)
    (:Added
      (CallFunction|CheckforNearbyEnemies false)
      (bind _returnvalue (Utilities|Time|SetTimerbyEvent (delegate Custom|LoopCheck) (Variables|Default|GetCheckInterval) true))
      (Variables|Default|SetTimer%20Check%20Enemies _returnvalue)
      (EventDispatchers|BindEventtoOnPlayerStartDBNO (Variables|GAVars|Data|GetOwningPlayerPawn) (EventDispatchers|CreateEvent :!event K2_CancelAbility)))))

(event Custom|LoopCheck
  ;[1 1 1] Run first check and start looping check
  (CallFunction|CheckforNearbyEnemies false))

(event Ability|EventOnEndAbility
  (params (bWasCancelled :type Boolean))
  ;[1 1 1] OnEndAbility: Kill timer and run one last check
  (Utilities|Time|ClearandInvalidateTimerbyHandle (Variables|Default|GetTimer%20Check%20Enemies))
  (CallFunction|CheckforNearbyEnemies false)
  (EventDispatchers|UnbindEventfromOnPlayerStartDBNO (Variables|GAVars|Data|GetOwningPlayerPawn) (EventDispatchers|CreateEvent :!event K2_CancelAbility)))

(fn CheckForNearbyEnemies
  (params (LastCheck1 :type Boolean))
  (bind _accumulated_success_time (Variables|Default|GetAccumulated%20Success%20Time))
  (bind _returnvalue (Math|Float|Floor _accumulated_success_time))
  (bind _owningplayerpawn (Variables|GAVars|Data|GetOwningPlayerPawn))
  ;[1 1 1] If not on the last check, save the time this loop triggered (used by last check)
  (if (not LastCheck1)
    (Variables|Default|SetLastLoopTime (Fortnite|Sync|Time|GetServerWorldTimeSeconds)))
  (Variables|Default|SetCurrent%20Check%20Succeeded false)
  ;[1 1 1] Check for nearby Fort Player Pawns
  (bind _outfortplayerpawns (Fortnite|GetFortPlayerPawnsInRange (Transformation|GetActorLocation (Variables|GAVars|Data|GetOwningPlayerPawn)) (* (Variables|Default|GetRadius%20in%20Meters) 1e+02)))
  ;[1 1 1] Check each hit actor until we find an enemy
  (bind _array_element (Utilities|Array|ForEachLoopwithBreak _outfortplayerpawns)
    (:LoopBody
      (bind _returnvalue_2 (GameplayTags|HasMatchingGameplayTag(Message) _array_element (Variables|Default|GetTag%20To%20Check)))
      (if _returnvalue_2
        (bind _asbp_player_pawn_non_participant (Utilities|Casting|CastToBP_PlayerPawn_NonParticipant _array_element))
        (switch Utilities|FlowControl|Switch|SwitchonEAlertLevel (Variables|/Game/Athena/AI/NPCs/Base/BP_PlayerPawn_NonParticipant.BP_PlayerPawn_NonParticipant_C|GetAlertLevel _asbp_player_pawn_non_participant)
          (:Unaware
            _)
          (:Suspicious
            _))
        (Variables|Default|SetCurrent%20Check%20Succeeded true)))
    (:Completed
      _))
  (if (and (Variables|Default|GetLast%20Check%20Succeeded) (Variables|Default|GetCurrent%20Check%20Succeeded))
    (Variables|Default|SetAccumulated%20Success%20Time (+ (Variables|Default|GetAccumulated%20Success%20Time) (select LastCheck1 (- (Fortnite|Sync|Time|GetServerWorldTimeSeconds) (Variables|Default|GetLastLoopTime)) (Variables|Default|GetCheckInterval))))
    (else
      (Ability|SendGameplayEventtoActor (Variables|GAVars|Data|GetOwningPlayerPawn) "(TagName=\"Athena.Quests.AI.NotInRange\")")))
  (Variables|Default|SetLast%20Check%20Succeeded (Variables|Default|GetCurrent%20Check%20Succeeded))
  ;[1 1 1] Update quest progress as we exceed int threshold, save off leftover successful time
  (if (Ability|DoesTargetHaveTags _owningplayerpawn (Variables|Default|GetTags))
    (if (>= (Variables|Default|GetAccumulated%20Success%20Time) 1.4)
      (Quest|SendComplexCustomStatEvent (Variables|GAVars|Data|GetOwningPlayerQuestManager) 0 (Variables|Default|GetAdditional%20Source%20Tags) "(GameplayTags=)" _returnvalue)
      (Variables|Default|SetAccumulated%20Success%20Time (- _accumulated_success_time (Math|Conversions|ToFloat(Integer) _returnvalue)))
      (bind _result (CustomVerb|AddTagsforSubject (CustomVerb|MakeCustomVerbMessage (Fortnite|GetFortPlayerControllerfromActor _owningplayerpawn) "(TagName=\"CustomVerb.CrouchNearEnemy\")") "Crouch" (GameplayTags|GetOwnedGameplayTags _owningplayerpawn :!class /Script/GameplayTags.GameplayTagAssetInterface)))
      (CustomVerb|BroadcastCustomVerbMessage _result)
      (Ability|SendGameplayEventtoActor 0 "(TagName=\"CustomVerb.CrouchNearEnemy\")"))
    (else
      ;[1 1 1] Stop Looping if Player is no longer crouching
      (Utilities|Time|ClearandInvalidateTimerbyHandle (Variables|Default|GetTimer%20Check%20Enemies)))))
